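# Sound cue for the Freeze Trap's destruction, using the Fort_Traps sound class and attenuation.
# A quality-level switch feeds four branches, each cheaper or equal to the one before.
# The first mixes random reused Ceiling Gas Trap destruction waves, quieter, under random pitch- and volume-modulated freeze explosions.
# The second uses one wave per layer, and the last two each play a single pre-optimized destroyed wave with modulation.
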SoundCue /Game/Sounds/Fort_Traps/Freeze_Trap/Freeze_Trap_Destroyed_Cue
root=QualityLevel_0
nodes [
 {id=Mixer_0, children [Modulator_4,Modulator_3], properties {inputVolume [0.35,1]}}
 {id=Mixer_1, children [Random_0,Modulator_0], properties {inputVolume [0.35,1]}}
 {id=Modulator_0, children [Random_1], properties {pitchMax=1.25, volumeMax=1.25}}
 {id=Modulator_1, children [WavePlayer_6], properties {pitchMax=1.15, pitchMin=0.9, volumeMax=1.1, volumeMin=0.92}}
 {id=Modulator_2, children [WavePlayer_7], properties {pitchMax=1.15, pitchMin=0.9, volumeMax=1.1, volumeMin=0.92}}
 {id=Modulator_3, children [WavePlayer_10], properties {pitchMax=1.25, volumeMax=1.25}}
 {id=Modulator_4, children [WavePlayer_9], properties {pitchMax=1.2, pitchMin=0.9}}
 {id=QualityLevel_0, children [Mixer_1,Mixer_0,Modulator_1,Modulator_2]}
 {id=Random_0, children [WavePlayer_0,WavePlayer_1,WavePlayer_2,WavePlayer_3], properties {weights [1,1,1,1]}}
 {id=Random_1, children [WavePlayer_4,WavePlayer_5], properties {weights [1,1]}}
 {id=WavePlayer_0, properties {soundWaveAssetPtr=/Game/Sounds/Fort_Traps/Ceiling_Gas_Trap/CeilingTrap_Gas_Destroyed_004}}
 {id=WavePlayer_1, properties {soundWaveAssetPtr=/Game/Sounds/Fort_Traps/Ceiling_Gas_Trap/CeilingTrap_Gas_Destroyed_003}}
 {id=WavePlayer_10, properties {soundWaveAssetPtr=&/Freeze_Trap_Explode_09}}
 {id=WavePlayer_2, properties {soundWaveAssetPtr=/Game/Sounds/Fort_Traps/Ceiling_Gas_Trap/CeilingTrap_Gas_Destroyed_002}}
 {id=WavePlayer_3, properties {soundWaveAssetPtr=/Game/Sounds/Fort_Traps/Ceiling_Gas_Trap/CeilingTrap_Gas_Destroyed_001}}
 {id=WavePlayer_4, properties {soundWaveAssetPtr=&/Freeze_Trap_Explode_10}}
 {id=WavePlayer_5, properties {soundWaveAssetPtr=&Freeze_Trap_Explode_09}}
 {id=WavePlayer_6, properties {soundWaveAssetPtr=&/Freeze_Trap_Destroyed_opt}}
 {id=WavePlayer_7, properties {soundWaveAssetPtr=&Freeze_Trap_Destroyed_opt}}
 {id=WavePlayer_9, properties {soundWaveAssetPtr=&CeilingTrap_Gas_Destroyed_004}}
]
properties {
 attenuationSettings=&Fort_Traps_Att
 maxDistance=6000
 soundClassObject=&Fort_Traps
}
