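# Sound cue for the Freeze Trap firing, which borrows gas-release waves from the Ceiling Gas Trap.
# A quality-level node picks one of four branches: the first layers a random gas wave with a pitch- and volume-modulated activation sound, and the second fixes the gas wave.
# The last two branches play only a pitch-varied optimized fire wave.
# Trap attenuation and sound class apply, and a concurrency override blocks new instances while one plays.
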SoundCue /Game/Sounds/Fort_Traps/Freeze_Trap/Freeze_Trap_Fire_Cue
root=QualityLevel_0
nodes [
 {id=Mixer_0, children [WavePlayer_1,Modulator_1], properties {inputVolume [0.75,1]}}
 {id=Mixer_1, children [Random_2,Modulator_0], properties {inputVolume [0.75,1]}}
 {id=Modulator_0, children [WavePlayer_9], properties {pitchMax=1.3, pitchMin=1, volumeMax=1.5, volumeMin=1}}
 {id=Modulator_1, children [WavePlayer_3], properties {pitchMax=1.3, pitchMin=1, volumeMax=1.5, volumeMin=1}}
 {id=Modulator_2, children [WavePlayer_0], properties {pitchMax=1.1, pitchMin=0.9}}
 {id=Modulator_3, children [WavePlayer_4], properties {pitchMax=1.1, pitchMin=0.9}}
 {id=QualityLevel_0, children [Mixer_1,Mixer_0,Modulator_2,Modulator_3]}
 {id=Random_2, children [WavePlayer_6,WavePlayer_7], properties {weights [1,1]}}
 {id=WavePlayer_0, properties {soundWaveAssetPtr=&/Freeze_Trap_Fire_opt}}
 {id=WavePlayer_1, properties {soundWaveAssetPtr=/Game/Sounds/Fort_Traps/Ceiling_Gas_Trap/CeilingTrap_Gas_Fire_002}}
 {id=WavePlayer_3, properties {soundWaveAssetPtr=&/Freeze_Trap_Fire_Activate_01}}
 {id=WavePlayer_4, properties {soundWaveAssetPtr=&Freeze_Trap_Fire_opt}}
 {id=WavePlayer_6, properties {soundWaveAssetPtr=&CeilingTrap_Gas_Fire_002}}
 {id=WavePlayer_7, properties {soundWaveAssetPtr=/Game/Sounds/Fort_Traps/Ceiling_Gas_Trap/CeilingTrap_Gas_Fire_001}}
 {id=WavePlayer_9, properties {soundWaveAssetPtr=&Freeze_Trap_Fire_Activate_01}}
]
properties {
 attenuationSettings=&Fort_Traps_Att
 bOverrideConcurrency=true
 maxDistance=6000
 soundClassObject=&Fort_Traps
 volumeMultiplier=0.7
 concurrencyOverrides {maxCount=1, resolutionRule=PreventNew}
}
