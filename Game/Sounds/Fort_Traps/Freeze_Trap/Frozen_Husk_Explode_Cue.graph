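# Sound cue for a frozen Husk exploding, filed under the Freeze Trap sounds and using Husk attenuation and an enemy-damaged sound class.
# A quality-level switch feeds four branches whose variation drops from three random waves to two, then to a single wave in the last two.
# Each branch modulates pitch and volume for variation.
# A concurrency override caps simultaneous instances at two.
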SoundCue /Game/Sounds/Fort_Traps/Freeze_Trap/Frozen_Husk_Explode_Cue
root=QualityLevel_0
nodes [
 {id=Modulator_0, children [Random_0], properties {pitchMax=1.2, pitchMin=1, volumeMax=1.25}}
 {id=Modulator_1, children [Random_1], properties {pitchMax=1.2, pitchMin=1, volumeMax=1.25}}
 {id=Modulator_2, children [WavePlayer_5], properties {pitchMax=1.2, pitchMin=1, volumeMax=1.25}}
 {id=Modulator_3, children [WavePlayer_6], properties {pitchMax=1.2, pitchMin=1, volumeMax=1.25}}
 {id=QualityLevel_0, children [Modulator_0,Modulator_1,Modulator_2,Modulator_3]}
 {id=Random_0, children [WavePlayer_1,WavePlayer_2,WavePlayer_3], properties {weights [1,1,1]}}
 {id=Random_1, children [WavePlayer_0,WavePlayer_4], properties {weights [1,1]}}
 {id=WavePlayer_0, properties {soundWaveAssetPtr=&/Frozen_Husk_Explode_01}}
 {id=WavePlayer_1, properties {soundWaveAssetPtr=&Frozen_Husk_Explode_01}}
 {id=WavePlayer_2, properties {soundWaveAssetPtr=&/Frozen_Husk_Explode_02}}
 {id=WavePlayer_3, properties {soundWaveAssetPtr=&/Frozen_Husk_Explode_03}}
 {id=WavePlayer_4, properties {soundWaveAssetPtr=&Frozen_Husk_Explode_02}}
 {id=WavePlayer_5, properties {soundWaveAssetPtr=&Frozen_Husk_Explode_03}}
 {id=WavePlayer_6, properties {soundWaveAssetPtr=&Frozen_Husk_Explode_03}}
]
properties {
 attenuationSettings=/Game/Sounds/Attenuation/Fort_Monster_Husk_Att
 bOverrideConcurrency=true
 maxDistance=2800
 soundClassObject=/Game/Sounds/FortSoundClassesAndModes/Fort_Enemy_Damaged_PSM
 volumeMultiplier=0.8
 concurrencyOverrides {maxCount=2}
}
